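# Sound cue for the thrust start of the SparkleRemix glider cosmetic.
# A Random node picks one of three equally weighted LoopIn wave variations.
# The cue plays at reduced volume with shared Athena glider thrust attenuation and concurrency settings, and is assigned to the Fort_Gliders_Land sound class.
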SoundCue /BRCosmetics/Sounds/Gliders/SparkleRemix/Glider_SparkleRemix_Thrust_Start_Cue
root=Random_1
nodes [
 {id=Random_1, children [WavePlayer_1,WavePlayer_3,WavePlayer_4], properties {weights [1,1,1]}}
 {id=WavePlayer_1, properties {soundWaveAssetPtr=&/Glider_SparkleRemix_LoopIn_01}}
 {id=WavePlayer_3, properties {soundWaveAssetPtr=&/Glider_SparkleRemix_LoopIn_02}}
 {id=WavePlayer_4, properties {soundWaveAssetPtr=&/Glider_SparkleRemix_LoopIn_03}}
]
properties {
 attenuationSettings=&Fort_Player_Glider_Thrust_Athena_Att
 maxDistance=12500
 soundClassObject=&Fort_Gliders_Land
 volumeMultiplier=0.2
 concurrencySet [&Athena_Glider_Thrust_Conc]
}
